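# Sound cue for a looping indoor room tone ambience in Athena.
# A single looping wave player plays a stereo small-space rumble at very low volume.
# It uses the Athena ambient attenuation and indoor ambient sound class, with a concurrency override that allows one instance and blocks new ones.
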SoundCue /Clyde/Athena/Sounds/Ambient/Amb_RoomTone_Athena_Cue
root=WavePlayer_0
nodes [{id=WavePlayer_0, properties {bLooping=true, soundWaveAssetPtr=/Clyde/ExtraDependencies/Rumble_SmallSpace001_Stereo_opt}}]
properties {
 attenuationSettings=&Fort_Ambient_Athena_Att
 maxDistance=2097152
 soundClassObject=&Fort_Ambient_Indoors_Athena
 volumeMultiplier=0.05
 concurrencyOverrides {maxCount=1, resolutionRule=PreventNew}
}
