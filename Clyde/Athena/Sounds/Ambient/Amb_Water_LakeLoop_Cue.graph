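# Looping lake water ambience cue for Athena, in the outdoor ambient sound class.
# A quality-level root selects among four branches: a random pick of two lake loops, a single full loop, or an "_opt" loop in the last two slots.
# Each branch runs through a modulator for slight pitch variation.
# Overridden attenuation disables spatialization, and concurrency settings specify one instance with stop-oldest resolution.
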SoundCue /Clyde/Athena/Sounds/Ambient/Amb_Water_LakeLoop_Cue
root=QualityLevel_0
nodes [
 {id=Modulator_0, children [Random_0], properties {pitchMax=1.1, pitchMin=0.9}}
 {id=Modulator_1, children [WavePlayer_1], properties {pitchMax=1.1, pitchMin=0.9}}
 {id=Modulator_2, children [WavePlayer_6], properties {pitchMax=1.1, pitchMin=0.9}}
 {id=Modulator_3, children [WavePlayer_7], properties {pitchMax=1.1, pitchMin=0.9}}
 {id=QualityLevel_0, children [Modulator_0,Modulator_1,Modulator_2,Modulator_3]}
 {id=Random_0, children [WavePlayer_3,WavePlayer_0], properties {weights [1,1]}}
 {id=WavePlayer_0, properties {bLooping=true, soundWaveAssetPtr=/Game/Sounds/Fort_Ambient_Looping/Water/water_lake_01b}}
 {id=WavePlayer_1, properties {bLooping=true, soundWaveAssetPtr=/Game/Sounds/Fort_Ambient_Looping/Water/water_lake_01a}}
 {id=WavePlayer_3, properties {bLooping=true, soundWaveAssetPtr=&water_lake_01a}}
 {id=WavePlayer_6, properties {bLooping=true, soundWaveAssetPtr=/Game/Sounds/Fort_Ambient_Looping/Water/water_lake_01a_opt}}
 {id=WavePlayer_7, properties {bLooping=true, soundWaveAssetPtr=&water_lake_01a_opt}}
]
properties {
 attenuationSettings=/Game/Athena/Sounds/Ambient/Fort_Amb_Water_Lake_Att
 bOverrideAttenuation=true
 maxDistance=2425
 soundClassObject=&Fort_Ambient_Outdoors_Athena
 volumeMultiplier=0.301
 attenuationOverrides {bSpatialize=false, falloffDistance=1500, nonSpatializedRadiusStart=700, stereoSpread=1000, attenuationShapeExtents {x=925}}
 concurrencyOverrides {maxCount=1, resolutionRule=StopOldest}
}
