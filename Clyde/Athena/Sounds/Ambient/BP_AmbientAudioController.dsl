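; Ambient audio controller that detects when the player's view target is enclosed by buildings and shifts the soundscape.
; A repeating cosmetic-only timer, skipped on BR mobile, runs five line traces (four sides and up), spread across frames with zero delays and aborting on the first miss.
; Inside/outside changes set a low-pass target and push or pop a sound mix.
; EndPlay fades out ambient loops; an indoor-loop toggle is defined but uncalled here.
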(event EventBeginPlay
  (Networking|CanExecuteCosmeticEvents
    (:True
      ;[1 1 1] Bind/Set-up timers and events
      (if (and (FortGlobals|IsBROnly) (TouchPlatform|IsMobileGame))
        (Components|Tick|SetComponentTickEnabled false)
        (else
          (Utilities|Time|SetTimerbyEvent (EventDispatchers|CreateEvent :!event Custom|OnCheckIfSurrounded) (Variables|Default|GetCheckFrequency) true))))))

(event EventEndPlay
  (params (EndPlayReason :type EEndPlayReason%20Enum :default "Destroyed"))
  (Utilities|FlowControl|Sequence
    (:then_0
      (bind _ambientmixa (Variables|Default|GetAmbientMixA)
        (:then
          (Audio|Components|Audio|FadeOut _ambientmixa 3 0.0 "Linear" :!class /Script/Engine.AudioComponent))
        (:else)))
    (:then_1
      (bind _ambientmixb (Variables|Default|GetAmbientMixB)
        (:then
          (Audio|Components|Audio|FadeOut _ambientmixb 3 0.0 "Linear" :!class /Script/Engine.AudioComponent))
        (:else)))
    (:then_2
      (bind _ambientmixindoor (Variables|Default|GetAmbientMixIndoor)
        (:then
          (Audio|Components|Audio|FadeOut _ambientmixindoor 3 0.0 "Linear" :!class /Script/Engine.AudioComponent))
        (:else)))
    (:then_3
      (Audio|PopSoundMixModifier 0))))

(event Custom|OnCheckIfSurrounded
  (bind _playercontroller (Variables|FortAmbientAudioController|GetPlayerController))
  (bind _returnvalue (Pawn|GetViewTarget _playercontroller))
  (bind _returnvalue_2 (Transformation|GetActorForwardVector _returnvalue))
  (bind _returnvalue_3 (Transformation|GetActorRightVector _returnvalue))
  (bind _returnvalue_4 (Transformation|GetActorLocation _returnvalue))
  ;[1 1 1] Check if player is inside/outside
  (Variables|Default|SetbIsInsidePrevious (Variables|Default|GetbIsInsideCurrent))
  (Variables|Default|SetbIsInsideCurrent false)
  (Utilities|IsValid (Pawn|GetViewTarget (Variables|FortAmbientAudioController|GetPlayerController))
    (:Is%20Valid
      (bind _returnvalue_5 (CallFunction|Trace _returnvalue_4 _returnvalue_2))
      (if _returnvalue_5
        (Utilities|FlowControl|Delay 0)
        (bind _returnvalue_6 (CallFunction|Trace _returnvalue_4 _returnvalue_3))
        (if _returnvalue_6
          (Utilities|FlowControl|Delay 0)
          (bind _returnvalue_7 (CallFunction|Trace _returnvalue_4 (* _returnvalue_2 -1)))
          (if _returnvalue_7
            (Utilities|FlowControl|Delay 0)
            (bind _returnvalue_8 (CallFunction|Trace _returnvalue_4 (* _returnvalue_3 -1)))
            (if _returnvalue_8
              (Utilities|FlowControl|Delay 0)
              (bind _returnvalue_9 (CallFunction|Trace _returnvalue_4 (Transformation|GetActorUpVector _returnvalue)))
              (if _returnvalue_9
                (Variables|Default|SetbIsInsideCurrent true)
                (CallFunction|EvaluateInsideState)
                (else
                  (CallFunction|EvaluateInsideState)))
              (else
                (CallFunction|EvaluateInsideState)))
            (else
              (CallFunction|EvaluateInsideState)))
          (else
            (CallFunction|EvaluateInsideState)))
        (else
          (CallFunction|EvaluateInsideState))))))

(event Custom|Evaluate%20Inside%20State
  ;[1 1 1] Check if player is inside/outside
  (if (Math|Boolean|NotEqual(Boolean) (Variables|Default|GetbIsInsideCurrent) (Variables|Default|GetbIsInsidePrevious))
    (if (Variables|Default|GetbIsInsideCurrent)
      (CallFunction|SetTargetLPFFreq (Variables|Default|GetInsideLPFFreq))
      (Audio|PushSoundMixModifier 0)
      (else
        (CallFunction|SetTargetLPFFreq (Variables|Default|GetOutsideLPFFreq))
        (Audio|PopSoundMixModifier 0)))))

(fn SetIndoorMixEnabled
  (params (Enabled :type Boolean))
  (if Enabled
    (bind _returnvalue (Audio|SpawnSoundatLocation (Variables|Default|GetLoop_Indoors)))
    (bind _output_get (Variables|Default|SetAmbientMixIndoor _returnvalue))
    (Audio|Components|Audio|FadeIn _output_get 1 1.0 0.0 "Linear" :!class /Script/Engine.AudioComponent)
    (else
      (bind _ambientmixindoor (Variables|Default|GetAmbientMixIndoor)
        (:then
          (Audio|Components|Audio|FadeOut _ambientmixindoor 1 0.0 "Linear" :!class /Script/Engine.AudioComponent))
        (:else)))))

(fn SetTargetLPFFreq
  (params (Target :type Float%20%28double-precision%29))
  (Variables|Default|SetTargetLPFFreq Target))

(fn Trace
  (params (Start :type Vector) (Direction :type Vector))
  (bind (_outhit _returnvalue) (Collision|LineTraceByChannel :Start Start :End (+ Start (* Direction 2304.0)) :TraceChannel "TraceTypeQuery10" :bTraceComplex false :DrawDebugType "None" :bIgnoreSelf true :TraceColor "(R=1.000000,G=0.000000,B=0.000000,A=1.000000)" :TraceHitColor "(R=0.000000,G=1.000000,B=0.000000,A=1.000000)" :DrawTime 3))
  (return (and _returnvalue (.bSuccess (Utilities|Casting|CastToBuildingActor (.HitActor (Collision|BreakHitResult _outhit)))))))
